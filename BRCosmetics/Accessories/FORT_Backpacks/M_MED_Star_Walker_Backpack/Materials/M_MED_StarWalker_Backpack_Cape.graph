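# Unlit, two-sided translucent material for the Star Walker backpack cape, drawing an animated galaxy and cracked-glass glow.
# A panning crack texture distorts a Fresnel-tinted nebula and drives a sine-pulsed glow plus a UV1 edge band.
# Two star layers, masked by faster panners to twinkle, multiply with the nebula before HDR emissive adjustment.
# Opacity combines a UV0 vertical gradient, inverted red vertex color and view-angle fade.
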Material /BRCosmetics/Accessories/FORT_Backpacks/M_MED_Star_Walker_Backpack/Materials/M_MED_StarWalker_Backpack_Cape
graph {
 properties {
  bUsedWithMeshDeformer=true
  bUsedWithSkeletalMesh=true
  blendMode=BLEND_Translucent
  pixelDepthOffsetMode=PDOM_Legacy
  shadingModel=MSM_Unlit
  twoSided=true
  usedShadingModels=Unlit
 }
 nodes [
  n1 = Multiply(A=Time(), B=ScalarParameter() {defaultValue=0.25})
  n2 = TextureCoordinate() {coordinateIndex=1, uTiling=2, vTiling=2}
  n3 = ComponentMask(None=TextureSample!(UVs=Panner(Coordinate=@n2) {speedY=-0.0772}, Tex, Level, Apply View MipBias) {mipValueMode=TMVM_MipLevel, texture=/Game/Effects/Fort_Effects/Textures/Tile/T_Glass_Cracks_M}.RGB) {r=true}
  n4 = Add(A=ComponentMask(None=TextureCoordinate() {coordinateIndex=1}) {g=true}, B=@n3)
  n5 = ConstantBiasScale(None=Sine(None=Add(A=@n1, B=@n4)))
  n6 = Multiply(
   A=VectorParameter() {parameterName=ColorG2, defaultValue=<rgba 0.2479 0.6136 1 1>}.RGB
   B=Power(Base=Multiply(A=@n5, B=@n3), Exp=ScalarParameter() {defaultValue=2, parameterName=FallofGlass})
  )
  n7 = ScalarParameter() {defaultValue=0.5, parameterName=Distort2}
  n8 = TextureCoordinate() {coordinateIndex=1}
  n9 = OneMinus(None=Add(A=Multiply(A=@n7, B=@n3), B=ComponentMask(None=@n8) {r=true}))
  n10 = ScalarParameter() {defaultValue=8, parameterName=FresnelPower3}
  n11 = Multiply(A=Saturate(None=Power(Base=@n9, Exp=@n10)), B=VectorParameter() {parameterName=ColorG1d, defaultValue=<rgba 0.1491 0.4393 1 1>}.RGB)
  n12 = VectorParameter() {parameterName=ColorG1, defaultValue=<rgba 0.0122 0.5449 1 1>}
  n13 = VectorParameter() {parameterName=ColorG1c, defaultValue=<rgba 0.3335 0.1875 2 1>}
  n14 = Fresnel(ExponentIn=ScalarParameter() {defaultValue=0.2314, parameterName=FresnelPower2}) {baseReflectFraction=0}
  n15 = TextureCoordinate() {coordinateIndex=1}
  n16 = Multiply(A=@n3, B=ScalarParameter() {defaultValue=0.5, parameterName=Distort})
  n17 = TextureSample(UVs=Panner(Coordinate=Add(A=@n15, B=@n16) {constB=0}) {speedY=-0.0625}) {mipValueMode=TMVM_MipLevel, texture=/Game/Characters/Player/Male/Medium/Bodies/M_Med_Soldier_04/Skins/Galaxy/Textures/T_FN_Nebula}
  n18 = LinearInterpolate(
   A=Multiply(A=LinearInterpolate(A=@n12.RGB, B=@n13.RGB, Alpha=@n14), B=ComponentMask(None=@n17.RGB) {r=true})
   B=Multiply(A=VectorParameter() {parameterName=ColorG1b, defaultValue=<rgba 0.0816 0.2477 0.3 1>}.RGB, B=ComponentMask(None=@n17.RGB) {b=true})
  )
  n19 = ScalarParameter() {defaultValue=0.0762, parameterName=BlendOriginal}
  n20 = TextureSample!(UVs=Panner(Coordinate=TextureCoordinate() {coordinateIndex=1, uTiling=4, vTiling=4}) {speedY=-0.0156}, Tex, Apply View MipBias) {samplerType=SAMPLERTYPE_Masks, texture=/Game/Characters/Player/Male/Medium/Bodies/M_Med_Soldier_04/Skins/Galaxy/Textures/Celestial_T_Stars}
  n21 = ComponentMask(None=@n20.RGB) {b=true}
  n22 = Multiply(A=@n21, B=@n21)
  n23 = TextureSample(UVs=Panner(Coordinate=TextureCoordinate() {uTiling=2, vTiling=2}) {speedX=0.137, speedY=-0.2}) {
   samplerSource=SSM_Wrap_WorldGroupSettings
   samplerType=SAMPLERTYPE_Masks
   texture=&Celestial_T_Stars
  }
  n24 = TextureSample(UVs=Panner(Coordinate=TextureCoordinate() {coordinateIndex=1, uTiling=2, vTiling=2}) {speedY=-0.0156}) {samplerType=SAMPLERTYPE_Masks, texture=&Celestial_T_Stars}
  n25 = ComponentMask(None=@n24.RGB) {r=true}
  n26 = Multiply(A=@n25, B=@n25)
  n27 = TextureSample(UVs=Panner(Coordinate=TextureCoordinate()) {speedX=0.137, speedY=-0.2}) {
   samplerSource=SSM_Wrap_WorldGroupSettings
   samplerType=SAMPLERTYPE_Masks
   texture=&Celestial_T_Stars
  }
  n28 = Add(
   A=Multiply(A=VectorParameter() {group=Galaxy, parameterName=Star Brightness_1, defaultValue=<rgba 10 10 10 1>}.RGB, B=Multiply(A=@n22, B=@n23.G))
   B=Multiply(A=VectorParameter() {group=Galaxy, parameterName=Star Brightness_2, defaultValue=<rgba 100 100 100 20>}.RGB, B=Multiply(A=@n26, B=@n27.G))
  )
  n29 = Add(
   A=Add(A=Add(A=@n6, B=@n11) {constB=0}, B=LinearInterpolate(A=@n18, B=@n17.RGB, Alpha=@n19))
   B=Multiply(A=VectorParameter() {parameterName=RiftShardColor, defaultValue=<rgba 0.9279 7.6847 10 0>}.RGB, B=Multiply(A=@n17.RGB, B=@n28))
  )
  MaterialFunctionCall(Emissive=MaterialFunctionCall(LDR Input=Clamp(None=@n29), HDR Tint=@n29, HDR Intensity Multiplier=ScalarParameter() {defaultValue=5, parameterName=HDR Intensity}) {materialFunction=&DeriveHDRfromLDR}.RemappedRange) {materialFunction=&MF_FX_EmissiveAdjustments}
  n30 = TextureCoordinate()
  n31 = Power(Base=OneMinus(None=ComponentMask(None=@n30) {g=true}), Exp=ScalarParameter() {defaultValue=4.5, parameterName=Alpha Contrast})
  n32 = ScalarParameter() {defaultValue=10, parameterName=Alpha Offset}
  n33 = OneMinus(None=VertexColor().R)
  n34 = ScalarParameter() {defaultValue=5, parameterName=Vertex Color Contrast}
  n35 = Saturate(None=Multiply(A=Power(Base=@n33, Exp=@n34), B=ScalarParameter() {defaultValue=2, parameterName=Vertex Color Offset}))
  n36 = MaterialFunctionCall("FadeSensitivity (optional)"=ScalarParameter() {defaultValue=1, parameterName=Fade sensitive}) {materialFunction=&FadeBasedOnViewAngle}
  Multiply(
   A=ScalarParameter() {defaultValue=1, parameterName=Alpha_Intensity}
   B=Multiply(A=Multiply(A=@n31, B=@n32), B=Multiply(A=@n35, B=@n36.Result))
  )
 ]
}
