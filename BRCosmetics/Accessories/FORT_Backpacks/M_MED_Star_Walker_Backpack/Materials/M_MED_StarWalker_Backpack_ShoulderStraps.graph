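# Translucent, unlit, two-sided material for the shoulder straps of the Star Walker backpack cosmetic, producing an animated glow.
# Panned noise indexes a color curve atlas row, whose intensified color is blended with a body emissive color and remapped to HDR.
# A UV1 gradient drives vertex offset scaled by world normals.
# Alpha combines a sine side mask, a squared vertical UV fade and view-angle fading.
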Material /BRCosmetics/Accessories/FORT_Backpacks/M_MED_Star_Walker_Backpack/Materials/M_MED_StarWalker_Backpack_ShoulderStraps
graph {
 properties {
  bUsedWithMeshDeformer=true
  bUsedWithSkeletalMesh=true
  blendMode=BLEND_Translucent
  pixelDepthOffsetMode=PDOM_Legacy
  shadingModel=MSM_Unlit
  twoSided=true
  usedShadingModels=Unlit
 }
 nodes [
  n1 = VectorParameter() {parameterName=Tile/Pan Direction, defaultValue=<rgba 1 1 0 -0.25>}
  n2 = Multiply(A=TextureCoordinate() {coordinateIndex=1}, B=AppendVector(A=@n1.R, B=@n1.R))
  n3 = Multiply(A=Time(), B=ScalarParameter() {defaultValue=0.5, parameterName=PanSpeed})
  n4 = AppendVector(A=@n1.B, B=@n1.B)
  n5 = CurveAtlasRowParameter(CurveTime=TextureSample(UVs=Panner(Coordinate=@n2, Time=@n3, Speed=@n4) {speedY=-0.04}) {samplerType=SAMPLERTYPE_Grayscale, texture=&LowResBlurredNoise}.G) {
   atlas=&AthenaCurveAtlas01
   curve=/Game/Characters/Player/Male/Medium/Bodies/M_MED_Star_Walker/Textures/Curve_Starwalker
   defaultValue=124
   parameterName=Color Curve
  }
  n6 = LinearInterpolate(
   A=VectorParameter() {parameterName=Body_Emissive_Color, defaultValue=<rgba 1 0 0.7902 1>}.RGB
   B=Multiply(A=@n5, B=ScalarParameter() {defaultValue=10, parameterName=Intensity})
  )
  MaterialFunctionCall(Emissive=MaterialFunctionCall(LDR Input=Saturate(None=@n6), HDR Tint=@n6, HDR Intensity Multiplier=ScalarParameter() {defaultValue=1, parameterName=HDR Intensity}) {materialFunction=&DeriveHDRfromLDR}.RemappedRange) {materialFunction=&MF_FX_EmissiveAdjustments}
  n7 = Add(A=TextureCoordinate() {coordinateIndex=1}) {constB=-0.5}
  n8 = Multiply(
   A=Power(Base=ComponentMask(None=@n7) {g=true}, Exp=ScalarParameter() {defaultValue=2, parameterName=WPO_Contrast})
   B=ScalarParameter() {defaultValue=4, parameterName=WPO_Mask_Intensity}
  ) {constB=2}
  Multiply(
   A=LinearInterpolate(B=VectorParameter() {parameterName=WPO_Amounts, defaultValue=<rgba 10 0 75 0>}.RGB, Alpha=Saturate(None=@n8))
   B=VertexNormalWS()
  )
  n9 = TextureCoordinate() {coordinateIndex=1}
  n10 = Power(Base=Sine(None=ComponentMask(None=@n9) {r=true}) {period=2}, Exp=ScalarParameter() {defaultValue=2.5, parameterName=Side Contrast})
  n11 = OneMinus(None=ComponentMask(None=TextureCoordinate()) {g=true})
  n12 = Multiply(A=@n11, B=@n11)
  n13 = ScalarParameter() {defaultValue=2, parameterName=Fade Sens}
  Multiply(
   A=ScalarParameter() {defaultValue=0.5, parameterName=Alpha Intensity}
   B=Multiply(
    A=Multiply(A=@n10, B=@n12)
    B=MaterialFunctionCall("FadeSensitivity (optional)"=@n13) {materialFunction=&FadeBasedOnViewAngle}.Result
   )
  )
 ]
}
