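# Physics asset for the Star Walker backpack cosmetic, simulating two mirrored dynamic tails on the left and right.
# Each tail is a four-bone chain with a small sphere at the root and capsules along the three following segments.
# Constraints link each segment to its predecessor with narrow first swing, wide second swing and limited twist, so the tails bend mostly along one axis.
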PhysicsAsset /BRCosmetics/Accessories/FORT_Backpacks/M_MED_Star_Walker_Backpack/Meshes/M_MED_Star_Walker_Pack_Physics
bodies {
 dyn_tail_01_l [{shapeType=Sphere, radius=1}]
 dyn_tail_01_r [{shapeType=Sphere, radius=1}]
 dyn_tail_02_l [{shapeType=Capsule, length=9, radius=4.75, center=<2 0 0>, rotation=<rot 90 0 0>}]
 dyn_tail_02_r [{shapeType=Capsule, length=9, radius=4.75, center=<2 0 0>, rotation=<rot 90 0 0>}]
 dyn_tail_03_l [{shapeType=Capsule, length=10.75, radius=5, center=<2.5 0 0>, rotation=<rot 90 0 0>}]
 dyn_tail_03_r [{shapeType=Capsule, length=10.75, radius=5, center=<2.5 0 0>, rotation=<rot 90 0 0>}]
 dyn_tail_04_l [{shapeType=Capsule, length=12, radius=4.5, center=<3 0 0>, rotation=<rot 90 0 0>}]
 dyn_tail_04_r [{shapeType=Capsule, length=12, radius=4.5, center=<3 0 0>, rotation=<rot 90 0 0>}]
}
constraints [
 {bone1Name=dyn_tail_03_l, bone2Name=dyn_tail_02_l, swing1LimitDegrees=5, swing2LimitDegrees=50, twistLimitDegrees=15}
 {bone1Name=dyn_tail_04_l, bone2Name=dyn_tail_03_l, swing1LimitDegrees=10, swing2LimitDegrees=80, twistLimitDegrees=10}
 {bone1Name=dyn_tail_02_l, bone2Name=dyn_tail_01_l, swing1LimitDegrees=5, swing2LimitDegrees=80, twistLimitDegrees=45}
 {bone1Name=dyn_tail_03_r, bone2Name=dyn_tail_02_r, swing1LimitDegrees=5, swing2LimitDegrees=50, twistLimitDegrees=15}
 {bone1Name=dyn_tail_04_r, bone2Name=dyn_tail_03_r, swing1LimitDegrees=10, swing2LimitDegrees=80, twistLimitDegrees=10}
 {bone1Name=dyn_tail_02_r, bone2Name=dyn_tail_01_r, swing1LimitDegrees=5, swing2LimitDegrees=80, twistLimitDegrees=45}
]
