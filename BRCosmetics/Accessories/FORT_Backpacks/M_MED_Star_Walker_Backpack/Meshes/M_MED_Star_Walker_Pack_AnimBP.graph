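# Animation Blueprint for the Star Walker backpack, which drives the secondary motion of its cape bones.
# The pose copied from the attached parent runs through chained Trail controllers for cape center and left/right forks, an additive root-bone tilt, then RigidBody simulation.
# Fork trails take wind velocity and alpha from variables not declared here.
# Output blends on bRigidBodyEnabled with a cached animation pose; backward-pose and angular-limit nodes sit unconnected to it.
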AnimBlueprint /BRCosmetics/Accessories/FORT_Backpacks/M_MED_Star_Walker_Backpack/Meshes/M_MED_Star_Walker_Pack_AnimBP
target_skeleton=/Game/Accessories/FORT_Backpacks/M_MED_Star_Walker_Backpack/Meshes/M_MED_Star_Walker_Pack_Skeleton
graphs [
 AnimGraph {nodes [
   Apply Limits!(ComponentPose:Component Space Pose Link Structure, Alpha:float=1, AngularOffsets_0:Vector=Get AimOffset() -> (AimOffset:Vector)) -> (Pose:Component Space Pose Link Structure) {
    is=title AnimGraphNode_ApplyLimits
    alphaBoolBlend {blendOption=Linear}
    angularOffsets [<0 0 0>]
    angularRangeLimits [{bone {boneName=dyn_cape01}, limitMax=<10 180 0>, limitMin=<-10 -180 -120>}]
   }
   Local Space Ref Pose()
   Local To Component(LocalPose=Use cached pose 'AnimCache'())
   n1 = Get bBlendInBackwardPose() -> (bBlendInBackwardPose:bool)
   Local To Component(LocalPose=Two Way Blend!(
     A=Component To Local()
     B=Cape_BackwardPose_PoseAsset : BackwardPose!(PoseAsset:Pose Asset Object Reference=/Game/Accessories/FORT_Backpacks/M_MED_Star_Walker_Backpack/Meshes/Cape_BackwardPose_PoseAsset) -> (Pose:Pose Link Structure) {is=title AnimGraphNode_PoseByName, poseName=BackwardPose}
     bAlphaBoolEnabled:bool=@n1
    ) -> (Pose) {blendNode {alphaInputType=Bool, alphaBoolBlend {blendInTime=0.5, blendOption=Linear, blendOutTime=0.5}}})
   n2 = Get FinalCapeBlendOutAlpha() -> (FinalCapeBlendOutAlpha:float)
   n3 = Get FinalCapeWindVelocity() -> (FinalCapeWindVelocity:Vector)
   n4 = Trail controller - Bone: dyn_cape05!(ComponentPose=Local To Component(LocalPose=Use cached pose 'Cape_Cache'()), Alpha=0.415, FakeVelocity:Vector="0,0,0") -> (Pose) {
    bActorSpaceFakeVel=true
    bUsePlanarLimit=true
    chainLength=5
    lastBoneRotationAnimAlphaBlend=0.5
    maxDeltaTime=0.0333
    relaxationSpeedScale=1.2
    alphaScaleBiasClamp {bInterpResult=true, interpSpeedDecreasing=15, interpSpeedIncreasing=15}
    baseJoint {boneName=backpack_root}
    fakeVelocity {y=50, z=-50}
    planarLimits [
     {drivingBone {boneName=backpack_root}, planeTransform {location=<0 -2.0567 0>, rotation=<rot 0 0 -100>, scale=<1 1 1>}}
     {drivingBone {boneName=backpack_root}, planeTransform {location=<0 0 10>, rotation=<rot 0 0 -150>, scale=<1 1 1>}}
    ]
    rotationLimits [{limitMax=<180 180 180>}, {limitMax=<180 180 180>}, {limitMax=<180 180 180>}, {limitMax=<180 180 180>}, {limitMax=<180 180 180>}]
    rotationOffsets [<0 0 0>, <0 0 0>, <0 0 0>, <0 0 0>, <0 0 0>]
    trailBone {boneName=dyn_cape05}
    trailRelaxationSpeed {editorCurveData {keys [<at 0 20 RCIM_Linear>, {arriveTangent=-14.4971, interpMode=RCIM_Linear, leaveTangent=-14.4971, time=1, value=10}]}}
   }
   n5 = Trail controller - Bone: dyn_cape06_l!(ComponentPose=@n4, Alpha=@n2, FakeVelocity:Vector=@n3) -> (Pose) {
    bActorSpaceFakeVel=true
    chainLength=6
    lastBoneRotationAnimAlphaBlend=1
    maxDeltaTime=0.0333
    baseJoint {boneName=backpack_root}
    planarLimits [{drivingBone {boneName=dyn_cape05}, planeTransform {location=<0 -2.0567 0>, rotation=<rot 0 0 -90>, scale=<1 1 1>}}]
    rotationLimits [
     {limitMax=<180 180 180>}
     {limitMax=<180 180 180>}
     {limitMax=<180 180 180>}
     {limitMax=<180 180 180>}
     {limitMax=<180 180 180>}
     {limitMax=<180 180 180>}
    ]
    rotationOffsets [<0 0 0>, <0 0 0>, <0 0 0>, <0 0 0>, <0 0 0>, <0 0 0>]
    trailBone {boneName=dyn_cape06_l}
    trailRelaxationSpeed {editorCurveData {keys [<at 0 15 RCIM_Linear>,<at 1 6 RCIM_Linear>]}}
   }
   n6 = Trail controller - Bone: dyn_cape06_r!(ComponentPose=@n5, Alpha=@n2, FakeVelocity:Vector=@n3) -> (Pose) {
    bActorSpaceFakeVel=true
    chainLength=6
    lastBoneRotationAnimAlphaBlend=1
    maxDeltaTime=0.0333
    baseJoint {boneName=backpack_root}
    planarLimits [{drivingBone {boneName=dyn_cape05}, planeTransform {location=<0 -2.0567 0>, rotation=<rot 0 0 -90>, scale=<1 1 1>}}]
    rotationLimits [
     {limitMax=<180 180 180>}
     {limitMax=<180 180 180>}
     {limitMax=<180 180 180>}
     {limitMax=<180 180 180>}
     {limitMax=<180 180 180>}
     {limitMax=<180 180 180>}
    ]
    rotationOffsets [<0 0 0>, <0 0 0>, <0 0 0>, <0 0 0>, <0 0 0>, <0 0 0>]
    trailBone {boneName=dyn_cape06_r}
    trailRelaxationSpeed {editorCurveData {keys [<at 0 15 RCIM_Linear>,<at 1 8 RCIM_Linear>]}}
   }
   n7 = Get bCanPlayCustomAnimations() -> (bCanPlayCustomAnimations:bool)
   n8 = "Transform (Modify) Bone - Bone: dyn_cape01"!(ComponentPose=@n6, Alpha="To Float (Boolean)"(InBool=NOT Boolean(A=@n7)), Rotation="0,0,-10") -> (Pose) {rotationMode=BMM_Additive, alphaBoolBlend {blendOption=Linear}, boneToModify {boneName=dyn_cape01}, rotation {roll=-10}}
   n9 = RigidBody(ComponentPose=@n8) {
    bClampLinearTranslationLimitToRefPose=true
    bOverrideWorldGravity=true
    overridePhysicsAsset=&/M_MED_Star_Walker_Pack_Physics
    simulationSpace=BaseBoneSpace
    alphaBoolBlend {blendOption=Linear}
    baseBoneRef {boneName=backpack_root}
    componentAppliedLinearAccClamp=<1000 1000 1000>
    componentLinearAccScale=<0.8 0.8 0.6>
    componentLinearVelScale=<0.8 0.8 0.6>
    overrideWorldGravity {z=-900}
   }
   Output Pose(Result=Blend Poses by bool(
     BlendPose_0=Component To Local(ComponentPose=@n9)
     BlendPose_1=Use cached pose 'AnimCache'()
     BlendTime_0=0.2
     BlendTime_1=0.2
     bActiveValue=Get bRigidBodyEnabled()
    ) {blendType=Linear}) {layerGroup=None}
   Save cached pose 'Cape_Cache'(Pose=Copy Pose From Mesh() {bUseAttachedParent=true}) {cacheName=Cape_Cache}
   Save cached pose 'AnimCache'(Pose=Blend Poses by bool(BlendPose_0=Use cached pose 'Cape_Cache'(), BlendPose_1=Component To Local(), BlendTime_0=0, BlendTime_1=0, bActiveValue=Get bClothEnabled()) {blendType=Linear}) {cacheName=AnimCache}
   n10 = Get bIsDBNO()
   SelectFloat(A=SelectFloat(A=0.5, B=Get BlendOutForksAlpha() -> (BlendOutForksAlpha:float), bPickA=@n10), B=1, bPickA=@n1)
   SelectVector(A=Get FakeWindVelocityDBNO() -> (FakeWindVelocityDBNO:Vector), B=Get FakeWindVelocity() -> (FakeWindVelocity:Vector), bPickA=@n10)
   n11 = Get bIsTargeting()
   SelectFloat(A=0.35, B=SelectFloat(A=SelectFloat(A=0, B=0.25, bPickA=@n11), B=1, bPickA=Get bIsCrouching()), bPickA=Get bIsFalling())
  ]}
 EventGraph {links [], nodes []}
]
variables []
